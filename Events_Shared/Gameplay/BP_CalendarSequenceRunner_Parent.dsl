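; Parent Blueprint for calendar event sequence runners, providing helpers that relate a Level Sequence's playback to calendar event timing.
; Timing comes from the calendar event, an optional active alternate event, or cheat-driven UTC values.
; When stretching is enabled, playrate is computed from remaining sequence seconds over remaining event time.
; Other helpers validate marked frames, convert frames to span ratios, forward the replicated ratio, and keep one persistent audio controller per world.
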(fn DoMarksExist
  (params (Marks :type Array%20of%20Strings))
  (if (== (Utilities|Array|Length Marks) 0)
    (return false)
    (else
      (LocalVariables|SetbMissingMark%20%28local%29 false)
      (Utilities|IsValid (Variables|Sequencer|GetLevel%20Sequence%20-%20Loaded)
        (:Is%20Valid
          (for _array_element Marks
            (if (not (.bFoundMark (CallFunction|GetFrameOfMark _array_element)))
              (Fortnite|FortShippingLog 0 (Utilities|String|Append (Utilities|String|Append (Variables|CalendarEventSequencePlayer|GetUnique%20String%20for%20Logging) "Missing Sequencer Mark: ") _array_element " in " (CallFunction|SequenceName)) true)
              (LocalVariables|SetbMissingMark%20%28local%29 true)))
          (return (not (LocalVariables|GetbMissingMark%20%28local%29))))
        (:Is%20Not%20Valid
          (Fortnite|FortShippingLog 0 (Utilities|String|Append (Variables|CalendarEventSequencePlayer|GetUnique%20String%20for%20Logging) "Do Marks Exist - Level Sequencer (Loaded) not valid.") true)
          (return true))) ; level sequence loaded and valid?
    )))

(fn GetCalendarEventInfo
  (params (bAllowStretchOverride :type Boolean))
  (switch Utilities|FlowControl|Switch|SwitchonStartingMethod_CalendarEventSequence_Enum (Variables|Cheats|GetStarting%20Method)
    (:NewEnumerator0
      (if (and (CallFunction|StretchOverrideValid) bAllowStretchOverride)
        (bind (_biseventactive _timesincebegin _timeuntilend _timespanratio) (CalendarEvents|GetCalendarEventInformation (Variables|CalendarEventSequencePlayer|GetAlternate%20Calendar%20Event%20for%20Stretch%20Timing)))
        (return _biseventactive _timesincebegin _timeuntilend _timespanratio)
        (else
          (bind (_biseventactive_2 _timesincebegin_2 _timeuntilend_2 _timespanratio_2) (CalendarEvents|GetCalendarEventInformation (Variables|CalendarEventSequencePlayer|GetCalendar%20Event%20Name)))
          (return _biseventactive_2 _timesincebegin_2 _timeuntilend_2 _timespanratio_2))))
    (:NewEnumerator1
      _)
    (:NewEnumerator4
      _))
  (return true (Math|DateTime|DateTime-DateTime (Math|DateTime|UTCNow) (Variables|Cheats|GetUTC%20Day%20Cheat%20Start%20Time)) (Math|DateTime|DateTime-DateTime (Math|DateTime|DateTime+Timespan (Variables|Cheats|GetUTC%20Day%20Cheat%20Start%20Time) (Math|Timespan|MakeTimespan 0 0 0 (Math|Float|Truncate (Variables|CalendarEventSequencePlayer|GetCheat%20Total%20Event%20Length%20%28Seconds%29)))) (Math|DateTime|UTCNow)) (Variables|Cheats|GetCheat%20Day%20Percentage)))

(fn GetFramefromSpanRatio
  (params (Span%20Ratio :type Float%20%28double-precision%29))
  (Utilities|IsValid (Variables|Sequencer|GetLevel%20Sequence%20-%20Loaded)
    (:Is%20Valid
      (bind _returnvalue (Sequencer|Sequence|GetPlaybackRange (Variables|Sequencer|GetLevel%20Sequence%20-%20Loaded)))
      (bind (_returnvalue_bhasstart _returnvalue_bhasend _returnvalue_inclusivestart _returnvalue_exclusiveend) (Utilities|Struct|BreakSequencerScriptingRange _returnvalue))
      (return (Math|Float|Truncate (Math|Float|Lerp (Math|Conversions|ToFloat(Integer) _returnvalue_inclusivestart) (Math|Conversions|ToFloat(Integer) _returnvalue_exclusiveend) Span%20Ratio))))
    (:Is%20Not%20Valid
      (Fortnite|FortShippingLog 0 (Utilities|String|Append (Variables|CalendarEventSequencePlayer|GetUnique%20String%20for%20Logging) "Get Frame From Span Ratio - Invalid Level Sequence") true))))

(fn GetFrameofMark
  (params (Key :type String%20%28by%20ref%29))
  (bind (_value _returnvalue) (Utilities|Map|Find (Variables|Sequencer|GetMarked%20Frames) Key))
  (return _value _returnvalue))

(fn GetScaledPlayrate
  (bind _sequence_player (Variables|Sequencer|GetSequence%20Player))
  (bind _returnvalue_frame (Utilities|TimeManagement|BreakQualifiedFrameTime (Sequencer|Player|GetCurrentTime _sequence_player)))
  (bind _returnvalue_frame_value (Utilities|Struct|BreakFrameNumber _returnvalue_frame))
  (bind _last_frame_in_sequence (Variables|Sequencer|GetLast%20Frame%20in%20Sequence))
  (bind (_time__seconds _total_frames) (CallFunction|GetTimeBetweenFrames _last_frame_in_sequence _returnvalue_frame_value))
  (if (Variables|CalendarEventSequencePlayer|GetbStretchSequenceToMatchCalendarEvent)
    (Utilities|IsValid (Variables|Sequencer|GetLevel%20Sequence%20-%20Loaded)
      (:Is%20Valid
        (Utilities|IsValid (Variables|Sequencer|GetSequence%20Player)
          (:Is%20Valid
            (Utilities|FlowControl|Sequence
              (:then_0
                (bind (_biseventactive _timesincebegin _timeuntilend) (CallFunction|GetCalendarEventInfo false))
                (LocalVariables|SetScaled%20Playrate%20for%20Remaining%20Calendar%20Event%20%28local%29 (/ _time__seconds (Math|Timespan|GetTotalSeconds _timeuntilend)))
                (Development|PrintString (Utilities|String|Append (Utilities|String|Append (Variables|CalendarEventSequencePlayer|GetUnique%20String%20for%20Logging) "Incremental Playrate Calculation - Remaining Frames ( ") (Utilities|String|ToString(Integer) _total_frames) " )" ", Remaining Seconds ( " (Utilities|String|ToString(Float) _time__seconds) " ), Incremental Scaled Playrate ( " (Utilities|String|ToString(Float) (LocalVariables|GetScaled%20Playrate%20for%20Remaining%20Calendar%20Event%20%28local%29)) " )" ", Original Scaled Playrate ( " (Utilities|String|ToString(Float) (Variables|Sequencer|GetFirst%20Calculated%20Playrate)) " )" ", remaining time ( " (Utilities|String|ToString(Text) (DisplayHelper|GetTimespanasSimpleString _timeuntilend)) " )") false false))
              (:then_1
                (bind (_biseventactive_2 _timesincebegin_2 _timeuntilend_2) (CallFunction|GetCalendarEventInfo false))
                (Utilities|String|LogString (Utilities|String|Append (Utilities|String|Append (Variables|CalendarEventSequencePlayer|GetUnique%20String%20for%20Logging) "Framerate ( ") (Utilities|String|ToString(Integer) (Variables|Sequencer|GetSequence%20Framerate)) " )" ", Total Frames ( " (Utilities|String|ToString(Integer) (Variables|Sequencer|GetTotal%20Sequence%20Frames)) " ), Event Length in Seconds ( " (Utilities|String|ToString(Float) (Math|Timespan|GetTotalSeconds (+ _timesincebegin_2 _timeuntilend_2))) " )" ", Sequence Length in Seconds ( " (Utilities|String|ToString(Float) (Variables|Sequencer|GetTotal%20Sequence%20Length%20%28seconds%29)) " )") false))
              (:then_2
                (return (LocalVariables|GetScaled%20Playrate%20for%20Remaining%20Calendar%20Event%20%28local%29)))))
          (:Is%20Not%20Valid
            (Fortnite|FortShippingLog 0 (Utilities|String|Append (Variables|CalendarEventSequencePlayer|GetUnique%20String%20for%20Logging) "Get Playrate from Frame - Invalid Level Sequence Player") true)
            (return 1.0))) ; level sequence is loaded?
      )
      (:Is%20Not%20Valid
        (Fortnite|FortShippingLog 0 (Utilities|String|Append (Variables|CalendarEventSequencePlayer|GetUnique%20String%20for%20Logging) "Get Playrate from Frame - Invalid Level Sequence") true)
        (return 1.0))) ; level sequence is loaded?
    (else
      (return 1.0))))

(fn GetSpanRatiofromFrame
  (params (Frame :type Integer))
  (Utilities|IsValid (Variables|Sequencer|GetLevel%20Sequence%20-%20Loaded)
    (:Is%20Valid
      (bind _returnvalue (Sequencer|Sequence|GetPlaybackRange (Variables|Sequencer|GetLevel%20Sequence%20-%20Loaded)))
      (bind (_returnvalue_bhasstart _returnvalue_bhasend _returnvalue_inclusivestart _returnvalue_exclusiveend) (Utilities|Struct|BreakSequencerScriptingRange _returnvalue))
      (if (or (> Frame _returnvalue_exclusiveend) (< Frame _returnvalue_inclusivestart))
        (Fortnite|FortShippingLog 0 (Utilities|String|Append (Variables|CalendarEventSequencePlayer|GetUnique%20String%20for%20Logging) "Get Span Ratio from Frame- Desired Frame Out of Range") true)
        (return 0.0)
        (else
          (return (/ (Math|Conversions|ToFloat(Integer) (- Frame _returnvalue_inclusivestart)) (Math|Conversions|ToFloat(Integer) (- _returnvalue_exclusiveend _returnvalue_inclusivestart)))))))
    (:Is%20Not%20Valid
      (Fortnite|FortShippingLog 0 (Utilities|String|Append (Variables|CalendarEventSequencePlayer|GetUnique%20String%20for%20Logging) "Get Span Ratio from Frame- Invalid Level Sequence") true))))

(fn GetSpanRatioofSequence
  (bind _returnvalue_frame (Utilities|TimeManagement|BreakQualifiedFrameTime (Sequencer|Player|GetCurrentTime (Variables|Sequencer|GetSequence%20Player))))
  (bind _returnvalue_frame_value (Utilities|Struct|BreakFrameNumber _returnvalue_frame))
  (Utilities|IsValid (Variables|Sequencer|GetSequence%20Player)
    (:Is%20Valid
      (Utilities|IsValid (Variables|Sequencer|GetLevel%20Sequence%20-%20Loaded)
        (:Is%20Valid
          (bind _returnvalue (Sequencer|Sequence|GetPlaybackRange (Variables|Sequencer|GetLevel%20Sequence%20-%20Loaded)))
          (bind (_returnvalue_bhasstart _returnvalue_bhasend _returnvalue_inclusivestart _returnvalue_exclusiveend) (Utilities|Struct|BreakSequencerScriptingRange _returnvalue))
          (return (/ (- _returnvalue_frame_value _returnvalue_inclusivestart) (- _returnvalue_exclusiveend _returnvalue_inclusivestart))))
        (:Is%20Not%20Valid
          (Fortnite|FortShippingLog 0 (Utilities|String|Append (Variables|CalendarEventSequencePlayer|GetUnique%20String%20for%20Logging) "Get Span Ratio of Sequence - Invalid Level Sequence") true))) ; level sequence loaded?
    )
    (:Is%20Not%20Valid
      (Fortnite|FortShippingLog 0 (Utilities|String|Append (Variables|CalendarEventSequencePlayer|GetUnique%20String%20for%20Logging) "Get Span Ratio of Sequence - Invalid Sequence Player") true)))
  (return 0.0))

(fn GetTimeBetweenFrames
  (params (Frame%201 :type Integer) (Frame%202 :type Integer))
  (return (Math|Fortnite|%28double%29integer64/%28double%29integer64 (Math|Conversions|ToInteger64(Integer) (Math|Integer|Absolute(Integer) (- Frame%201 Frame%202))) (Math|Conversions|ToInteger64(Integer) (Variables|Sequencer|GetSequence%20Framerate))) (- Frame%201 Frame%202)))

(fn GetTimeBetweenMarks
  (params (Frame%201 :type String) (Frame%202 :type String))
  (bind _ballmarksexist (CallFunction|DoMarksExist (Utilities|Array|MakeArray Frame%201 Frame%202)))
  (if _ballmarksexist
    (return (CallFunction|GetTimeBetweenFrames (CallFunction|GetFrameOfMark Frame%201) (CallFunction|GetFrameOfMark Frame%202)) true)
    (else
      (return 0.0 false))))

(fn InitializeAudioController
  ;[1 1 1] Spawn only one audio controller per world
  (bind _returnvalue (Actor|GetActorOfClass "/Events_Shared/Sound/BP/BP_Events_PersistentAudio_Controller.BP_Events_PersistentAudio_Controller_C"))
  (Utilities|IsValid _returnvalue
    (:Is%20Valid
      (Variables|Default|SetPersistentAudioController _returnvalue))
    (:Is%20Not%20Valid
      (bind _returnvalue_2 (Game|SpawnActorfromClass "/Events_Shared/Sound/BP/BP_Events_PersistentAudio_Controller.BP_Events_PersistentAudio_Controller_C" (Transformation|GetActorTransform) "Undefined" "MultiplyWithRoot" 0 0))
      (Variables|Default|SetPersistentAudioController _returnvalue_2))))

(fn OnRep_CurrentSequencerRatio
  (CallFunction|CurrentSequenceRatioUpdated (Variables|Replicated|GetCurrent%20Sequencer%20Ratio)))

(fn SequenceName
  (bind _sequence_player (Variables|Sequencer|GetSequence%20Player))
  (Utilities|IsValid _sequence_player
    (:Is%20Valid
      (return (Sequencer|Player|GetSequenceName _sequence_player false :!class /Script/MovieScene.MovieSceneSequencePlayer)))
    (:Is%20Not%20Valid
      (return "Invalid Sequence"))))

(fn StretchOverrideValid
  (if (Utilities|String|IsEmpty (Variables|CalendarEventSequencePlayer|GetAlternate%20Calendar%20Event%20for%20Stretch%20Timing))
    (return false)
    (else
      (bind _biseventactive (CalendarEvents|GetCalendarEventInformation (Variables|CalendarEventSequencePlayer|GetAlternate%20Calendar%20Event%20for%20Stretch%20Timing)))
      (return _biseventactive))))

(fn ConstructionScript)
