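; Event volume actor that provides toggleable fake blocking collision and pushes players and vehicles away.
; BeginPlay uses ScalableFloat settings to toggle the blocking capsule's collision and a gated, server-side repulsion pass.
; That pass launches overlapping player pawns outward and upward, optionally ignoring fall damage, and can repeat after a delay because one overlap trigger proved unreliable.
; Vehicles entering the repulsing capsule are launched server-side using reversed velocity plus outward impulse.
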(event Custom|Enable%20Fake%20Collision
  ;[0.1958 0.5098 1] Repulsion
  ;;[1 1 1] Fake Collision
  (Collision|SetCollisionEnabled (Variables|Default|GetCapsule%20-%20Blocking) "QueryAndPhysics") ; Enable Fake Collision
)

(event EventBeginPlay
  ;[0.1958 0.5098 1] Repulsion
  ;;[1 1 1] Begin Play
  (Utilities|FlowControl|Sequence
    (:then_0
      (if (ScalableFloat|GetValueasBool (Variables|FakeCollision&Repulsion|GetRepulsion%20Enabled%20on%20Begin%20Play) 1.0)
        (CallFunction|EnableRepulsion)
        (else
          (CallFunction|DisableRepulsion))))
    (:then_1
      (if (ScalableFloat|GetValueasBool (Variables|FakeCollision&Repulsion|GetFake%20Collision%20Enabled%20on%20Begin%20Play) 1.0)
        (CallFunction|EnableFakeCollision)
        (else
          (CallFunction|DisableFakeCollision))))) ; 0 - Repulsion\n1 - Collision
)

(event Custom|Disable%20Fake%20Collision
  ;[0.1958 0.5098 1] Repulsion
  ;;[1 1 1] Fake Collision
  (Collision|SetCollisionEnabled (Variables|Default|GetCapsule%20-%20Blocking) "NoCollision") ; Disable Fake Collision
)

(event Custom|Disable%20Repulsion
  ;[0.1958 0.5098 1] Repulsion
  ;;[1 1 1] Impulse Players
  (Utilities|FlowControl|Gate false
    (:Exit
      (for _array_element (Collision|GetOverlappingActors (Variables|Default|GetCapsule%20-%20Repulsing) "/Script/FortniteGame.FortPlayerPawn" :!class /Script/Engine.PrimitiveComponent)
        (bind _asfort_player_pawn (Utilities|Casting|CastToFortPlayerPawn _array_element))
        (Variables|/Script/FortniteGame.FortPawn|SetbIgnoreNextFallingDamage (ScalableFloat|GetValueasBool (Variables|FakeCollision&Repulsion|GetRepulsed%20Player%20Ignores%20Next%20Fall%20Damage) 1.0) _asfort_player_pawn) ; Player should Ignore Next Fall Damage?\n\n..for events, we really want to avoid eliminating\nplayers if possible or they might miss the event.\n\nSo, normally whenever impulsing the player or\nputting them in a situation where they might fall,\nwe just have them ignore the next fall damage.
        ;;;[1 1 1] Impulse Player
        (Fortnite|FortShippingLog 0 "Impulsing Player")
        (bind (_returnvalue_x _returnvalue_y) (Math|Vector|BreakVector (- (Transformation|GetActorLocation _array_element) (Transformation|GetWorldLocation (Variables|Default|GetCapsule%20-%20Repulsing)))))
        (Pawn|Character|LaunchCharacterJump _asfort_player_pawn (+ (* (Math|Vector|Normalize (Math|Vector|MakeVector _returnvalue_x _returnvalue_y)) (Math|Vector|MakeVector (ScalableFloat|GetValueatLevel (Variables|FakeCollision&Repulsion|GetRepulsion%20Impulse%20-%20Horizontal) 1.0) (ScalableFloat|GetValueatLevel (Variables|FakeCollision&Repulsion|GetRepulsion%20Impulse%20-%20Horizontal) 1.0))) (Math|Vector|MakeVector 0.0 0.0 (ScalableFloat|GetValueatLevel (Variables|FakeCollision&Repulsion|GetRepulsion%20Impulse%20-%20Vertical) 1.0))) true true true true) ; Players who are standing still aren't being\nlaunched on launcher with Add Impulse, so\ntrying to use this node to get the player moving\nbefore impulsing them.
        (CallFunction|OnPlayerImpulsed _asfort_player_pawn))
      ;;;[1 1 1] Re-Trigger Impulse
      (if (ScalableFloat|GetValueasBool (Variables|FakeCollision&Repulsion|GetRepulsion%20Continuously%20Triggers) 1.0)
        (Utilities|FlowControl|Delay (ScalableFloat|GetValueatLevel (Variables|FakeCollision&Repulsion|GetRepulsion%20Re-Trigger%20Delay) 1.0)) ; Delay then impulse again. This has\nproven more reliable than just using\nPlayer Overlap and triggering once in\ncase the impulse doesn't throw the player\nfar enough. Just keep trying to impulse.
      ))))

(event Custom|Enable%20Repulsion
  ;[0.1958 0.5098 1] Repulsion
  ;;[1 1 1] Impulse Players
  (Utilities|FlowControl|Sequence
    (:then_0
      (Utilities|FlowControl|Gate false
        (:Exit
          _)))
    (:then_1
      (Networking|SwitchHasAuthority
        (:Authority
          _)) ; Just find and impulse players\non the server as local clients\nshouldn't be initiating things like\nmovement changes on players.
    ))
  (for _array_element (Collision|GetOverlappingActors (Variables|Default|GetCapsule%20-%20Repulsing) "/Script/FortniteGame.FortPlayerPawn" :!class /Script/Engine.PrimitiveComponent)
    (bind _asfort_player_pawn (Utilities|Casting|CastToFortPlayerPawn _array_element))
    (Variables|/Script/FortniteGame.FortPawn|SetbIgnoreNextFallingDamage (ScalableFloat|GetValueasBool (Variables|FakeCollision&Repulsion|GetRepulsed%20Player%20Ignores%20Next%20Fall%20Damage) 1.0) _asfort_player_pawn) ; Player should Ignore Next Fall Damage?\n\n..for events, we really want to avoid eliminating\nplayers if possible or they might miss the event.\n\nSo, normally whenever impulsing the player or\nputting them in a situation where they might fall,\nwe just have them ignore the next fall damage.
    ;;;[1 1 1] Impulse Player
    (Fortnite|FortShippingLog 0 "Impulsing Player")
    (bind (_returnvalue_x _returnvalue_y) (Math|Vector|BreakVector (- (Transformation|GetActorLocation _array_element) (Transformation|GetWorldLocation (Variables|Default|GetCapsule%20-%20Repulsing)))))
    (Pawn|Character|LaunchCharacterJump _asfort_player_pawn (+ (* (Math|Vector|Normalize (Math|Vector|MakeVector _returnvalue_x _returnvalue_y)) (Math|Vector|MakeVector (ScalableFloat|GetValueatLevel (Variables|FakeCollision&Repulsion|GetRepulsion%20Impulse%20-%20Horizontal) 1.0) (ScalableFloat|GetValueatLevel (Variables|FakeCollision&Repulsion|GetRepulsion%20Impulse%20-%20Horizontal) 1.0))) (Math|Vector|MakeVector 0.0 0.0 (ScalableFloat|GetValueatLevel (Variables|FakeCollision&Repulsion|GetRepulsion%20Impulse%20-%20Vertical) 1.0))) true true true true) ; Players who are standing still aren't being\nlaunched on launcher with Add Impulse, so\ntrying to use this node to get the player moving\nbefore impulsing them.
    (CallFunction|OnPlayerImpulsed _asfort_player_pawn))
  ;;;[1 1 1] Re-Trigger Impulse
  (if (ScalableFloat|GetValueasBool (Variables|FakeCollision&Repulsion|GetRepulsion%20Continuously%20Triggers) 1.0)
    (Utilities|FlowControl|Delay (ScalableFloat|GetValueatLevel (Variables|FakeCollision&Repulsion|GetRepulsion%20Re-Trigger%20Delay) 1.0)) ; Delay then impulse again. This has\nproven more reliable than just using\nPlayer Overlap and triggering once in\ncase the impulse doesn't throw the player\nfar enough. Just keep trying to impulse.
  ))

(event OnComponentBeginOverlap(Capsule-Repulsing)
  (params (OverlappedComponent :type Primitive%20Component%20Object%20Reference) (OtherActor :type Actor%20Object%20Reference) (OtherComp :type Primitive%20Component%20Object%20Reference) (OtherBodyIndex :type Integer) (bFromSweep :type Boolean) (SweepResult :type Hit%20Result%20Structure%20%28by%20ref%29))
  ;[0.1958 0.5098 1] Repulsion
  ;;[0.6025 1 0.9809] Launch vehicle
  (Networking|SwitchHasAuthority
    (:Authority
      (bind _asfort_vehicle_interface (Utilities|Casting|CastToFortVehicleInterface OtherActor))
      (bind _asfort_athena_vehicle (Utilities|Casting|CastToFortAthenaVehicle (Utilities|ToObject(Interface) _asfort_vehicle_interface)))
      (bind (_returnvalue_x _returnvalue_y) (Math|Vector|BreakVector (Math|Vector|Normalize (- (Class|FortVehicleInterface|GetOwnerLocation _asfort_vehicle_interface) (Transformation|GetActorLocation)))))
      (bind (_returnvalue_x_2 _returnvalue_y_2) (Math|Vector|BreakVector (Utilities|Operators|Multiply (Math|Vector|MakeVector _returnvalue_x _returnvalue_y) (Math|Conversions|ToVector(Float) (ScalableFloat|GetValueatLevel (Variables|FakeCollision&Repulsion|GetVehicle%20Impulse%20Multiplier))))))
      (Class|VehicleFunctionLibrary|LaunchVehicle _asfort_athena_vehicle (+ (* (Vehicle|GetVehicleVelocity _asfort_athena_vehicle) (Math|Conversions|ToVector(Float) -1.0)) (Math|Vector|MakeVector _returnvalue_x_2 _returnvalue_y_2 (ScalableFloat|GetValueatLevel (Variables|FakeCollision&Repulsion|GetVehicle%20Impulse%20Vertical))))))))

(fn LogIt
  (params (InString :type String) (bLogAsWarning :type Boolean))
  (Fortnite|FortShippingLog 0 (Utilities|String|Append "[BP_FakeCollisionVolume_and_PlayerRepulsor] - " InString) bLogAsWarning))

(fn OnPlayerImpulsed
  (params (ImpulsedPawn :type Fort%20Player%20Pawn%20Object%20Reference))
  (return ImpulsedPawn))

(fn ConstructionScript)
