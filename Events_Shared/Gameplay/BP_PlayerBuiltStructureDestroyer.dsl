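; Event actor that clears player-built structures inside a box sized by its actor scale, keeping players from interfering with live events.
; Begin Play (optional) or a custom start event runs a server-side, gated box-overlap pass that destroys player-built BuildingActors, optionally repeating on a delay until stopped.
; Each pass also flags pawns in a scaled, minimum-clamped box to ignore their next fall damage.
; ScalableFloats toggle and tune each behavior.
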(event EventBeginPlay
  (bind _self self)
  (bind _returnvalue (Transformation|GetActorScale3D _self))
  (bind _returnvalue_2 (Utilities|String|ToString(Vector) _returnvalue))
  (bind _returnvalue_3 (Utilities|String|Append "[PistolCanary] PlayerBuiltStructure Destroyer started at size: " _returnvalue_2))
  (bind (_returnvalue_x _returnvalue_y _returnvalue_z) (Math|Vector|BreakVector (* (Transformation|GetActorScale3D self) (Math|Vector|MakeVector (ScalableFloat|GetValueatLevel (Variables|PlayerBuiltStructureDestroyer|GetFall%20Damage%20Distance%20Multiplier%20of%20Box%20Extent%20Scale) 1.0) (ScalableFloat|GetValueatLevel (Variables|PlayerBuiltStructureDestroyer|GetFall%20Damage%20Distance%20Multiplier%20of%20Box%20Extent%20Scale) 1.0) (ScalableFloat|GetValueatLevel (Variables|PlayerBuiltStructureDestroyer|GetFall%20Damage%20Distance%20Multiplier%20of%20Box%20Extent%20Scale) 1.0)))))
  ;[1 1 1] Destroy Player Built Structures & Avoid Fall Damage
  (if (ScalableFloat|GetValueasBool (Variables|PlayerBuiltStructureDestroyer|GetStart%20Destroying%20Structures%20on%20Begin%20Play) 1.0)
    (Networking|SwitchHasAuthority
      (:Authority
        (Utilities|FlowControl|Sequence
          (:then_0
            _)
          (:then_1
            _)) ; 0 - open the gate\n\n1 - start destroying player \nbuilt structures before delaying
        (Utilities|FlowControl|Gate false
          (:Exit
            (Utilities|FlowControl|Sequence
              (:then_0
                _)
              (:then_1
                (if (ScalableFloat|GetValueasBool (Variables|PlayerBuiltStructureDestroyer|GetShould%20Continuously%20Destroy%20Structures) 1.0)
                  (Utilities|FlowControl|Delay (ScalableFloat|GetValueatLevel (Variables|PlayerBuiltStructureDestroyer|GetContinous%20Destruction%20Re-Trigger%20Delay) 1.0)) ; Keep destroying structures on a timer to \nreally keep players from interfering with\nthe event.
                ))
              (:then_2
                ;;[1 1 1] Player Fall Damage
                (if (> (ScalableFloat|GetValueatLevel (Variables|PlayerBuiltStructureDestroyer|GetShould%20Falling%20Players%20Ignore%20Fall%20Damage) 1.0) 0)
                  (bind _outactors (Collision|BoxOverlapActors (Transformation|GetActorLocation self) (Math|Vector|MakeVector (Math|Float|Max(Float) _returnvalue_x (ScalableFloat|GetValueatLevel (Variables|PlayerBuiltStructureDestroyer|GetMinimum%20Extent%20Check%20for%20Saving%20Players) 1.0)) (Math|Float|Max(Float) _returnvalue_y (ScalableFloat|GetValueatLevel (Variables|PlayerBuiltStructureDestroyer|GetMinimum%20Extent%20Check%20for%20Saving%20Players) 1.0)) (Math|Float|Max(Float) _returnvalue_z (ScalableFloat|GetValueatLevel (Variables|PlayerBuiltStructureDestroyer|GetMinimum%20Extent%20Check%20for%20Saving%20Players) 1.0))) (Utilities|Array|MakeArray "ObjectTypeQuery3") "/Script/FortniteGame.FortPlayerPawn" :!class /Script/Engine.KismetSystemLibrary)) ; find players who may be falling after\nplayer built structures are destroyed
                  (for _array_element _outactors
                    (bind _asfort_player_pawn (Utilities|Casting|CastToFortPlayerPawn _array_element))
                    (Variables|/Script/FortniteGame.FortPawn|SetbIgnoreNextFallingDamage true _asfort_player_pawn) ; we  don't know if they are up in the air or not,\nbut for this event we just want players to not\ndie when falling from their structures
                    (Development|PrintText (Utilities|Text|FormatText "NSLOCTEXT(\"[2E433DAEA58EF7F5278769D9B46CA05A]\", \"67EA63DD44AF2E5121AD76B72529E24C\", \"Ignore Next Fall Damage: {player}\")" (Variables|/Script/FortniteGame.FortPawn|GetDisplayName _asfort_player_pawn)))
                    (Fortnite|FortShippingLog 0 (Utilities|String|ToString(Text) (Utilities|Text|FormatText "NSLOCTEXT(\"[2E433DAEA58EF7F5278769D9B46CA05A]\", \"67EA63DD44AF2E5121AD76B72529E24C\", \"Ignore Next Fall Damage: {player}\")" (Variables|/Script/FortniteGame.FortPawn|GetDisplayName _asfort_player_pawn)))))))) ; 0 - destroy player built buildings!\n1 - delay and do it again\n2 - Player Fall Damage
            ;;[1 1 1] Destroy Player Built Structures
            (if (> (ScalableFloat|GetValueatLevel (Variables|PlayerBuiltStructureDestroyer|GetShould%20Break%20Player%20Built%20Structures) 1.0) 0)
              (Development|PrintString _returnvalue_3)
              (Fortnite|FortShippingLog 0 _returnvalue_3)
              (bind _outactors_2 (Collision|BoxOverlapActors (Transformation|GetActorLocation _self) (/ _returnvalue "2.000000,2.000000,2.000000") (Utilities|Array|MakeArray "ObjectTypeQuery1") "/Script/FortniteGame.BuildingActor" :!class /Script/Engine.KismetSystemLibrary)) ; find player built structures
              (for _array_element_2 _outactors_2
                (bind _asbuilding_actor (Utilities|Casting|CastToBuildingActor _array_element_2))
                (if (BuildingActor|IsPlayerBuilt _asbuilding_actor)
                  (Actor|DestroyActor _asbuilding_actor) ; Destroy the player made building actor
                  (Development|PrintText (Utilities|Text|FormatText "NSLOCTEXT(\"[2E433DAEA58EF7F5278769D9B46CA05A]\", \"A74C34D445E929CF1F585CA8D081672A\", \"DESTROY  player built: {building}\")" (Utilities|GetDisplayName _asbuilding_actor)) false)
                  (Fortnite|FortShippingLog 0 (Utilities|String|ToString(Text) (Utilities|Text|FormatText "NSLOCTEXT(\"[2E433DAEA58EF7F5278769D9B46CA05A]\", \"A74C34D445E929CF1F585CA8D081672A\", \"DESTROY  player built: {building}\")" (Utilities|GetDisplayName _asbuilding_actor)))))))))))))

(event Custom|Start%20Destroying%20Player%20Built%20Structures
  (bind _self self)
  (bind _returnvalue (Transformation|GetActorScale3D _self))
  (bind _returnvalue_2 (Utilities|String|ToString(Vector) _returnvalue))
  (bind _returnvalue_3 (Utilities|String|Append "[PistolCanary] PlayerBuiltStructure Destroyer started at size: " _returnvalue_2))
  (bind (_returnvalue_x _returnvalue_y _returnvalue_z) (Math|Vector|BreakVector (* (Transformation|GetActorScale3D self) (Math|Vector|MakeVector (ScalableFloat|GetValueatLevel (Variables|PlayerBuiltStructureDestroyer|GetFall%20Damage%20Distance%20Multiplier%20of%20Box%20Extent%20Scale) 1.0) (ScalableFloat|GetValueatLevel (Variables|PlayerBuiltStructureDestroyer|GetFall%20Damage%20Distance%20Multiplier%20of%20Box%20Extent%20Scale) 1.0) (ScalableFloat|GetValueatLevel (Variables|PlayerBuiltStructureDestroyer|GetFall%20Damage%20Distance%20Multiplier%20of%20Box%20Extent%20Scale) 1.0)))))
  ;[1 1 1] Destroy Player Built Structures & Avoid Fall Damage
  (Networking|SwitchHasAuthority
    (:Authority
      (Utilities|FlowControl|Sequence
        (:then_0
          _)
        (:then_1
          _)) ; 0 - open the gate\n\n1 - start destroying player \nbuilt structures before delaying
      (Utilities|FlowControl|Gate false
        (:Exit
          (Utilities|FlowControl|Sequence
            (:then_0
              _)
            (:then_1
              (if (ScalableFloat|GetValueasBool (Variables|PlayerBuiltStructureDestroyer|GetShould%20Continuously%20Destroy%20Structures) 1.0)
                (Utilities|FlowControl|Delay (ScalableFloat|GetValueatLevel (Variables|PlayerBuiltStructureDestroyer|GetContinous%20Destruction%20Re-Trigger%20Delay) 1.0)) ; Keep destroying structures on a timer to \nreally keep players from interfering with\nthe event.
              ))
            (:then_2
              ;;[1 1 1] Player Fall Damage
              (if (> (ScalableFloat|GetValueatLevel (Variables|PlayerBuiltStructureDestroyer|GetShould%20Falling%20Players%20Ignore%20Fall%20Damage) 1.0) 0)
                (bind _outactors (Collision|BoxOverlapActors (Transformation|GetActorLocation self) (Math|Vector|MakeVector (Math|Float|Max(Float) _returnvalue_x (ScalableFloat|GetValueatLevel (Variables|PlayerBuiltStructureDestroyer|GetMinimum%20Extent%20Check%20for%20Saving%20Players) 1.0)) (Math|Float|Max(Float) _returnvalue_y (ScalableFloat|GetValueatLevel (Variables|PlayerBuiltStructureDestroyer|GetMinimum%20Extent%20Check%20for%20Saving%20Players) 1.0)) (Math|Float|Max(Float) _returnvalue_z (ScalableFloat|GetValueatLevel (Variables|PlayerBuiltStructureDestroyer|GetMinimum%20Extent%20Check%20for%20Saving%20Players) 1.0))) (Utilities|Array|MakeArray "ObjectTypeQuery3") "/Script/FortniteGame.FortPlayerPawn" :!class /Script/Engine.KismetSystemLibrary)) ; find players who may be falling after\nplayer built structures are destroyed
                (for _array_element _outactors
                  (bind _asfort_player_pawn (Utilities|Casting|CastToFortPlayerPawn _array_element))
                  (Variables|/Script/FortniteGame.FortPawn|SetbIgnoreNextFallingDamage true _asfort_player_pawn) ; we  don't know if they are up in the air or not,\nbut for this event we just want players to not\ndie when falling from their structures
                  (Development|PrintText (Utilities|Text|FormatText "NSLOCTEXT(\"[2E433DAEA58EF7F5278769D9B46CA05A]\", \"67EA63DD44AF2E5121AD76B72529E24C\", \"Ignore Next Fall Damage: {player}\")" (Variables|/Script/FortniteGame.FortPawn|GetDisplayName _asfort_player_pawn)))
                  (Fortnite|FortShippingLog 0 (Utilities|String|ToString(Text) (Utilities|Text|FormatText "NSLOCTEXT(\"[2E433DAEA58EF7F5278769D9B46CA05A]\", \"67EA63DD44AF2E5121AD76B72529E24C\", \"Ignore Next Fall Damage: {player}\")" (Variables|/Script/FortniteGame.FortPawn|GetDisplayName _asfort_player_pawn)))))))) ; 0 - destroy player built buildings!\n1 - delay and do it again\n2 - Player Fall Damage
          ;;[1 1 1] Destroy Player Built Structures
          (if (> (ScalableFloat|GetValueatLevel (Variables|PlayerBuiltStructureDestroyer|GetShould%20Break%20Player%20Built%20Structures) 1.0) 0)
            (Development|PrintString _returnvalue_3)
            (Fortnite|FortShippingLog 0 _returnvalue_3)
            (bind _outactors_2 (Collision|BoxOverlapActors (Transformation|GetActorLocation _self) (/ _returnvalue "2.000000,2.000000,2.000000") (Utilities|Array|MakeArray "ObjectTypeQuery1") "/Script/FortniteGame.BuildingActor" :!class /Script/Engine.KismetSystemLibrary)) ; find player built structures
            (for _array_element_2 _outactors_2
              (bind _asbuilding_actor (Utilities|Casting|CastToBuildingActor _array_element_2))
              (if (BuildingActor|IsPlayerBuilt _asbuilding_actor)
                (Actor|DestroyActor _asbuilding_actor) ; Destroy the player made building actor
                (Development|PrintText (Utilities|Text|FormatText "NSLOCTEXT(\"[2E433DAEA58EF7F5278769D9B46CA05A]\", \"A74C34D445E929CF1F585CA8D081672A\", \"DESTROY  player built: {building}\")" (Utilities|GetDisplayName _asbuilding_actor)) false)
                (Fortnite|FortShippingLog 0 (Utilities|String|ToString(Text) (Utilities|Text|FormatText "NSLOCTEXT(\"[2E433DAEA58EF7F5278769D9B46CA05A]\", \"A74C34D445E929CF1F585CA8D081672A\", \"DESTROY  player built: {building}\")" (Utilities|GetDisplayName _asbuilding_actor))))))))))))

(event Custom|DestroyPlayerBuilt
  (bind _self self)
  (bind _returnvalue (Transformation|GetActorScale3D _self))
  (bind _returnvalue_2 (Utilities|String|ToString(Vector) _returnvalue))
  (bind _returnvalue_3 (Utilities|String|Append "[PistolCanary] PlayerBuiltStructure Destroyer started at size: " _returnvalue_2))
  (bind (_returnvalue_x _returnvalue_y _returnvalue_z) (Math|Vector|BreakVector (* (Transformation|GetActorScale3D self) (Math|Vector|MakeVector (ScalableFloat|GetValueatLevel (Variables|PlayerBuiltStructureDestroyer|GetFall%20Damage%20Distance%20Multiplier%20of%20Box%20Extent%20Scale) 1.0) (ScalableFloat|GetValueatLevel (Variables|PlayerBuiltStructureDestroyer|GetFall%20Damage%20Distance%20Multiplier%20of%20Box%20Extent%20Scale) 1.0) (ScalableFloat|GetValueatLevel (Variables|PlayerBuiltStructureDestroyer|GetFall%20Damage%20Distance%20Multiplier%20of%20Box%20Extent%20Scale) 1.0)))))
  ;[1 1 1] Destroy Player Built Structures & Avoid Fall Damage
  (Networking|SwitchHasAuthority
    (:Authority
      (Utilities|FlowControl|Sequence
        (:then_0
          _)
        (:then_1
          _)) ; 0 - open the gate\n\n1 - start destroying player \nbuilt structures before delaying
      (Utilities|FlowControl|Gate false
        (:Exit
          (Utilities|FlowControl|Sequence
            (:then_0
              _)
            (:then_1
              (if (ScalableFloat|GetValueasBool (Variables|PlayerBuiltStructureDestroyer|GetShould%20Continuously%20Destroy%20Structures) 1.0)
                (Utilities|FlowControl|Delay (ScalableFloat|GetValueatLevel (Variables|PlayerBuiltStructureDestroyer|GetContinous%20Destruction%20Re-Trigger%20Delay) 1.0)) ; Keep destroying structures on a timer to \nreally keep players from interfering with\nthe event.
              ))
            (:then_2
              ;;[1 1 1] Player Fall Damage
              (if (> (ScalableFloat|GetValueatLevel (Variables|PlayerBuiltStructureDestroyer|GetShould%20Falling%20Players%20Ignore%20Fall%20Damage) 1.0) 0)
                (bind _outactors (Collision|BoxOverlapActors (Transformation|GetActorLocation self) (Math|Vector|MakeVector (Math|Float|Max(Float) _returnvalue_x (ScalableFloat|GetValueatLevel (Variables|PlayerBuiltStructureDestroyer|GetMinimum%20Extent%20Check%20for%20Saving%20Players) 1.0)) (Math|Float|Max(Float) _returnvalue_y (ScalableFloat|GetValueatLevel (Variables|PlayerBuiltStructureDestroyer|GetMinimum%20Extent%20Check%20for%20Saving%20Players) 1.0)) (Math|Float|Max(Float) _returnvalue_z (ScalableFloat|GetValueatLevel (Variables|PlayerBuiltStructureDestroyer|GetMinimum%20Extent%20Check%20for%20Saving%20Players) 1.0))) (Utilities|Array|MakeArray "ObjectTypeQuery3") "/Script/FortniteGame.FortPlayerPawn" :!class /Script/Engine.KismetSystemLibrary)) ; find players who may be falling after\nplayer built structures are destroyed
                (for _array_element _outactors
                  (bind _asfort_player_pawn (Utilities|Casting|CastToFortPlayerPawn _array_element))
                  (Variables|/Script/FortniteGame.FortPawn|SetbIgnoreNextFallingDamage true _asfort_player_pawn) ; we  don't know if they are up in the air or not,\nbut for this event we just want players to not\ndie when falling from their structures
                  (Development|PrintText (Utilities|Text|FormatText "NSLOCTEXT(\"[2E433DAEA58EF7F5278769D9B46CA05A]\", \"67EA63DD44AF2E5121AD76B72529E24C\", \"Ignore Next Fall Damage: {player}\")" (Variables|/Script/FortniteGame.FortPawn|GetDisplayName _asfort_player_pawn)))
                  (Fortnite|FortShippingLog 0 (Utilities|String|ToString(Text) (Utilities|Text|FormatText "NSLOCTEXT(\"[2E433DAEA58EF7F5278769D9B46CA05A]\", \"67EA63DD44AF2E5121AD76B72529E24C\", \"Ignore Next Fall Damage: {player}\")" (Variables|/Script/FortniteGame.FortPawn|GetDisplayName _asfort_player_pawn)))))))) ; 0 - destroy player built buildings!\n1 - delay and do it again\n2 - Player Fall Damage
          ;;[1 1 1] Destroy Player Built Structures
          (if (> (ScalableFloat|GetValueatLevel (Variables|PlayerBuiltStructureDestroyer|GetShould%20Break%20Player%20Built%20Structures) 1.0) 0)
            (Development|PrintString _returnvalue_3)
            (Fortnite|FortShippingLog 0 _returnvalue_3)
            (bind _outactors_2 (Collision|BoxOverlapActors (Transformation|GetActorLocation _self) (/ _returnvalue "2.000000,2.000000,2.000000") (Utilities|Array|MakeArray "ObjectTypeQuery1") "/Script/FortniteGame.BuildingActor" :!class /Script/Engine.KismetSystemLibrary)) ; find player built structures
            (for _array_element_2 _outactors_2
              (bind _asbuilding_actor (Utilities|Casting|CastToBuildingActor _array_element_2))
              (if (BuildingActor|IsPlayerBuilt _asbuilding_actor)
                (Actor|DestroyActor _asbuilding_actor) ; Destroy the player made building actor
                (Development|PrintText (Utilities|Text|FormatText "NSLOCTEXT(\"[2E433DAEA58EF7F5278769D9B46CA05A]\", \"A74C34D445E929CF1F585CA8D081672A\", \"DESTROY  player built: {building}\")" (Utilities|GetDisplayName _asbuilding_actor)) false)
                (Fortnite|FortShippingLog 0 (Utilities|String|ToString(Text) (Utilities|Text|FormatText "NSLOCTEXT(\"[2E433DAEA58EF7F5278769D9B46CA05A]\", \"A74C34D445E929CF1F585CA8D081672A\", \"DESTROY  player built: {building}\")" (Utilities|GetDisplayName _asbuilding_actor))))))))))))

(event Custom|Stop%20Destroying%20Player%20Built%20Structures
  (bind _self self)
  (bind _returnvalue (Transformation|GetActorScale3D _self))
  (bind _returnvalue_2 (Utilities|String|ToString(Vector) _returnvalue))
  (bind _returnvalue_3 (Utilities|String|Append "[PistolCanary] PlayerBuiltStructure Destroyer started at size: " _returnvalue_2))
  (bind (_returnvalue_x _returnvalue_y _returnvalue_z) (Math|Vector|BreakVector (* (Transformation|GetActorScale3D self) (Math|Vector|MakeVector (ScalableFloat|GetValueatLevel (Variables|PlayerBuiltStructureDestroyer|GetFall%20Damage%20Distance%20Multiplier%20of%20Box%20Extent%20Scale) 1.0) (ScalableFloat|GetValueatLevel (Variables|PlayerBuiltStructureDestroyer|GetFall%20Damage%20Distance%20Multiplier%20of%20Box%20Extent%20Scale) 1.0) (ScalableFloat|GetValueatLevel (Variables|PlayerBuiltStructureDestroyer|GetFall%20Damage%20Distance%20Multiplier%20of%20Box%20Extent%20Scale) 1.0)))))
  ;[1 1 1] Destroy Player Built Structures & Avoid Fall Damage
  (Utilities|FlowControl|Gate false
    (:Exit
      (Utilities|FlowControl|Sequence
        (:then_0
          _)
        (:then_1
          (if (ScalableFloat|GetValueasBool (Variables|PlayerBuiltStructureDestroyer|GetShould%20Continuously%20Destroy%20Structures) 1.0)
            (Utilities|FlowControl|Delay (ScalableFloat|GetValueatLevel (Variables|PlayerBuiltStructureDestroyer|GetContinous%20Destruction%20Re-Trigger%20Delay) 1.0)) ; Keep destroying structures on a timer to \nreally keep players from interfering with\nthe event.
          ))
        (:then_2
          ;;[1 1 1] Player Fall Damage
          (if (> (ScalableFloat|GetValueatLevel (Variables|PlayerBuiltStructureDestroyer|GetShould%20Falling%20Players%20Ignore%20Fall%20Damage) 1.0) 0)
            (bind _outactors (Collision|BoxOverlapActors (Transformation|GetActorLocation self) (Math|Vector|MakeVector (Math|Float|Max(Float) _returnvalue_x (ScalableFloat|GetValueatLevel (Variables|PlayerBuiltStructureDestroyer|GetMinimum%20Extent%20Check%20for%20Saving%20Players) 1.0)) (Math|Float|Max(Float) _returnvalue_y (ScalableFloat|GetValueatLevel (Variables|PlayerBuiltStructureDestroyer|GetMinimum%20Extent%20Check%20for%20Saving%20Players) 1.0)) (Math|Float|Max(Float) _returnvalue_z (ScalableFloat|GetValueatLevel (Variables|PlayerBuiltStructureDestroyer|GetMinimum%20Extent%20Check%20for%20Saving%20Players) 1.0))) (Utilities|Array|MakeArray "ObjectTypeQuery3") "/Script/FortniteGame.FortPlayerPawn" :!class /Script/Engine.KismetSystemLibrary)) ; find players who may be falling after\nplayer built structures are destroyed
            (for _array_element _outactors
              (bind _asfort_player_pawn (Utilities|Casting|CastToFortPlayerPawn _array_element))
              (Variables|/Script/FortniteGame.FortPawn|SetbIgnoreNextFallingDamage true _asfort_player_pawn) ; we  don't know if they are up in the air or not,\nbut for this event we just want players to not\ndie when falling from their structures
              (Development|PrintText (Utilities|Text|FormatText "NSLOCTEXT(\"[2E433DAEA58EF7F5278769D9B46CA05A]\", \"67EA63DD44AF2E5121AD76B72529E24C\", \"Ignore Next Fall Damage: {player}\")" (Variables|/Script/FortniteGame.FortPawn|GetDisplayName _asfort_player_pawn)))
              (Fortnite|FortShippingLog 0 (Utilities|String|ToString(Text) (Utilities|Text|FormatText "NSLOCTEXT(\"[2E433DAEA58EF7F5278769D9B46CA05A]\", \"67EA63DD44AF2E5121AD76B72529E24C\", \"Ignore Next Fall Damage: {player}\")" (Variables|/Script/FortniteGame.FortPawn|GetDisplayName _asfort_player_pawn)))))))) ; 0 - destroy player built buildings!\n1 - delay and do it again\n2 - Player Fall Damage
      ;;[1 1 1] Destroy Player Built Structures
      (if (> (ScalableFloat|GetValueatLevel (Variables|PlayerBuiltStructureDestroyer|GetShould%20Break%20Player%20Built%20Structures) 1.0) 0)
        (Development|PrintString _returnvalue_3)
        (Fortnite|FortShippingLog 0 _returnvalue_3)
        (bind _outactors_2 (Collision|BoxOverlapActors (Transformation|GetActorLocation _self) (/ _returnvalue "2.000000,2.000000,2.000000") (Utilities|Array|MakeArray "ObjectTypeQuery1") "/Script/FortniteGame.BuildingActor" :!class /Script/Engine.KismetSystemLibrary)) ; find player built structures
        (for _array_element_2 _outactors_2
          (bind _asbuilding_actor (Utilities|Casting|CastToBuildingActor _array_element_2))
          (if (BuildingActor|IsPlayerBuilt _asbuilding_actor)
            (Actor|DestroyActor _asbuilding_actor) ; Destroy the player made building actor
            (Development|PrintText (Utilities|Text|FormatText "NSLOCTEXT(\"[2E433DAEA58EF7F5278769D9B46CA05A]\", \"A74C34D445E929CF1F585CA8D081672A\", \"DESTROY  player built: {building}\")" (Utilities|GetDisplayName _asbuilding_actor)) false)
            (Fortnite|FortShippingLog 0 (Utilities|String|ToString(Text) (Utilities|Text|FormatText "NSLOCTEXT(\"[2E433DAEA58EF7F5278769D9B46CA05A]\", \"A74C34D445E929CF1F585CA8D081672A\", \"DESTROY  player built: {building}\")" (Utilities|GetDisplayName _asbuilding_actor))))))))))

(fn ConstructionScript)
