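; Special event zone that applies gameplay effects to overlapping player pawns and EMPs or knocks back vehicles.
; On the authority, calendar events, routed string events and pre-event messages drive ZoneActive and PreEventActive.
; RepNotifies toggle zone visuals and collision, and during pre-event extend effects to all players, with an optional client HUD banner.
; Granted-tag checks prevent duplicate effects; leaving the area removes them unless pre-event is active.
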(event EventBeginPlay
  (Networking|SwitchHasAuthority
    (:Authority
      (Class|FortAsyncActionAuthorityMatchReady|MatchReadyAsync
        (:OnReady
          (MCPContext|InitialCalendarSyncCompleted
            (:then
              (bind (_asynctaskproxy _payload) (GameplayEventRouter|StartListeningToEvent (GameplayEventRouter|GetGlobalEventRouter) "/Events_Shared/Gameplay/GlobalEvents/Event_RouterEvent_String.Event_RouterEvent_String")
                (:then
                  (bind (_asynctaskproxy_2 _payload_2) (GameplayEventRouter|StartListeningToEvent (GameplayEventRouter|GetGlobalEventRouter) "/Game/Athena/Events/Generic/CountdownTimer/PreEventStruct/GenericPreEventStruct.GenericPreEventStruct")
                    (:OnEventReceived
                      (if (Utilities|String|EqualExactly(String) (Utilities|Struct|BreakGenericPreEventStruct _payload_2) (Variables|Default|GetPreEventActiveName))
                        (Variables|Default|SetPreEventActive true)
                        (Variables|Default|SetZoneActive false)
                        (elif (Utilities|String|EqualExactly(String) (Utilities|Struct|BreakGenericPreEventStruct _payload_2) (Variables|Default|GetDeactivateEventName))
                          (Variables|Default|SetPreEventActive false))))))
                (:OnEventReceived
                  (if (Utilities|String|EqualExactly(String) (Utilities|Struct|BreakEventRouterEventString _payload) (Variables|Default|GetActivateEventName))
                    (Variables|Default|SetZoneActive true)
                    (elif (Utilities|String|EqualExactly(String) (Utilities|Struct|BreakEventRouterEventString _payload) (Variables|Default|GetDeactivateEventName))
                      (Variables|Default|SetZoneActive false))))))
            (:OnSyncComplete
              (for _array_element (Variables|Default|GetZoneCalendarEvents)
                (MCPContext|WatchCalendarEvent _array_element
                  (:OnEventActive
                    (Variables|Default|SetZoneActive true)))))))))))

(event Custom|PonyBackerPreEventMessageTest
  (Variables|Default|SetPreEventActive true)
  (Variables|Default|SetZoneActive false))

(event OnComponentBeginOverlap(AreaOfEffect)
  (params (OverlappedComponent :type Primitive%20Component%20Object%20Reference) (OtherActor :type Actor%20Object%20Reference) (OtherComp :type Primitive%20Component%20Object%20Reference) (OtherBodyIndex :type Integer) (bFromSweep :type Boolean) (SweepResult :type Hit%20Result%20Structure%20%28by%20ref%29))
  (Networking|SwitchHasAuthority
    (:Authority
      (bind _asfort_player_pawn (Utilities|Casting|CastToFortPlayerPawn OtherActor)
        (:then
          (if (Pawn|IsPlayerControlled _asfort_player_pawn)
            (Utilities|IsValid (Ability|GetFortAbilitySystemComponent _asfort_player_pawn)
              (:Is%20Valid
                (for _array_element (Variables|Default|GetAreaGameplayEffects)
                  (if (not (Default|DoesPlayerHaveEffect _array_element (Ability|GetFortAbilitySystemComponent _asfort_player_pawn)))
                    (GameplayEffects|ApplyGameplayEffectToSelf :self (Ability|GetFortAbilitySystemComponent _asfort_player_pawn) :GameplayEffectClass _array_element :Level 0.0)))
                (Development|PrintString "Player has overlapped!")))))
        (:CastFailed
          (bind _asfort_athena_vehicle (Utilities|Casting|CastToFortAthenaVehicle OtherActor))
          (CallFunction|ApplyVehicleEffects _asfort_athena_vehicle))))))

(event OnComponentEndOverlap(AreaOfEffect)
  (params (OverlappedComponent :type Primitive%20Component%20Object%20Reference) (OtherActor :type Actor%20Object%20Reference) (OtherComp :type Primitive%20Component%20Object%20Reference) (OtherBodyIndex :type Integer))
  (Networking|SwitchHasAuthority
    (:Authority
      (if (not (Variables|Default|GetPreEventActive))
        (bind _asfort_player_pawn (Utilities|Casting|CastToFortPlayerPawn OtherActor)
          (:then
            (if (Pawn|IsPlayerControlled _asfort_player_pawn)
              (for _array_element (Variables|Default|GetAreaGameplayEffects)
                (GameplayEffects|RemoveActiveGameplayEffectbySourceEffect (Ability|GetFortAbilitySystemComponent _asfort_player_pawn) _array_element 0 -1))
              (Development|PrintString "Player has left!!" true true "(R=0.000000,G=0.660000,B=1.000000,A=1.000000)" 5.0)))
          (:CastFailed
            (bind _asfort_athena_vehicle (Utilities|Casting|CastToFortAthenaVehicle OtherActor))
            (CallFunction|RemoveVehicleEffects _asfort_athena_vehicle)))))))

(event Custom|RemoveAreaEffects
  (Networking|SwitchHasAuthority
    (:Authority
      (for _array_element (Collision|GetOverlappingActors (Variables|Default|GetAreaOfEffect) "/Script/FortniteGame.FortPlayerPawn" :!class /Script/Engine.PrimitiveComponent)
        (for _array_element_2 (Variables|Default|GetAreaGameplayEffects)
          (GameplayEffects|RemoveActiveGameplayEffectbySourceEffect (Ability|GetFortAbilitySystemComponent _array_element) _array_element_2 0 -1))
        (Development|PrintString "Effects have been removed")))))

(event Custom|SE_ActivateZone
  (Variables|Default|SetZoneActive true))

(event Custom|SE_DeactivateZone
  (Variables|Default|SetZoneActive false))

(event Custom|ApplyVehicleEffects
  (params (OverlappingVehicle :type Fort%20Athena%20Vehicle%20Object%20Reference))
  (if (Variables|Default|GetVehicleEMPActive)
    (CallFunction|ApplyEMPEffect OverlappingVehicle true)
    (CallFunction|ApplyKnockback OverlappingVehicle)))

(event Custom|RemoveVehicleEffects
  (params (OverlappingVehicle :type Fort%20Athena%20Vehicle%20Object%20Reference))
  (CallFunction|ApplyEMPEffect OverlappingVehicle))

(fn ApplyEMPEffect
  (params (VehicleRef :type Fort%20Athena%20Vehicle%20Object%20Reference) (ApplyEMP :type Boolean))
  (bind _returnvalue (Vehicle|GetVehicleAbilitySystemComponent VehicleRef))
  (if ApplyEMP
    (Vehicle|WakeUp VehicleRef)
    (Components|Movement|StopMovementImmediately (Pawn|GetMovementComponent VehicleRef))
    (GameplayEffects|ApplyGameplayEffectToSelf :self _returnvalue :GameplayEffectClass (Variables|Default|GetEMPEffect) :Level 0.0)
    (else
      (GameplayEffects|RemoveActiveGameplayEffectbySourceEffect _returnvalue (Variables|Default|GetEMPEffect) 0 -1))))

(fn ApplyKnockback
  (params (VehicleRef :type Fort%20Athena%20Vehicle%20Object%20Reference))
  (if (Class|FortVehicleInterface|IsBoosting VehicleRef)
    (Physics|AddRadialForce (Variables|/Script/FortniteGame.FortAthenaVehicle|GetSkeletalMesh VehicleRef) (Transformation|GetActorLocation self) 777.0 5e+01 "RIF_Constant" true) ; Apply Force to cancel out boost.
  )
  (Audio|BroadcastAbilityImpactatLocation 0 VehicleRef (CallFunction|GetImpulseValue VehicleRef) 777.0))

(fn DoesPlayerHaveEffect
  (params (EffectToApply :type Gameplay%20Effect%20Class%20Reference) (PlayerAbilitySystem :type Fort%20Ability%20System%20Component%20Object%20Reference))
  (LocalVariables|SetFoundTag false)
  (bind _array_element (Utilities|Array|ForEachLoopwithBreak (GameplayTags|BreakGameplayTagContainer (GameplayTags|GetOwnedGameplayTags PlayerAbilitySystem :!class /Script/GameplayTags.BlueprintGameplayTagLibrary)))
    (:LoopBody
      (for _array_element_2 (GameplayTags|BreakGameplayTagContainer (Ability|GameplayEffect|GetGrantedTags EffectToApply :!class /Script/GameplayAbilities.AbilitySystemBlueprintLibrary))
        (if (== (GameplayTags|GetTagName _array_element_2) (GameplayTags|GetTagName _array_element))
          (Development|PrintString "Player Has This Ability")
          (LocalVariables|SetFoundTag true))))
    (:Completed
      _))
  (return (LocalVariables|GetFoundTag)))

(fn GetImpulseValue
  (params (OtherActor :type Actor%20Object%20Reference))
  (bind (_returnvalue_x _returnvalue_y) (Math|Vector|BreakVector (- (Transformation|GetActorLocation OtherActor) (Transformation|GetActorLocation self))))
  (Utilities|IsValid OtherActor
    (:Is%20Valid
      (return (+ (* (Math|Vector|Normalize (Math|Vector|MakeVector _returnvalue_x _returnvalue_y)) (Math|Vector|MakeVector (ScalableFloat|GetValueatLevel (Variables|Default|GetImpulse%20Vehicle%20-%20Horizontal) 1.0) (ScalableFloat|GetValueatLevel (Variables|Default|GetImpulse%20Vehicle%20-%20Horizontal) 1.0))) (Math|Vector|MakeVector 0.0 0.0 (* (ScalableFloat|GetValueatLevel (Variables|Default|GetImpulse%20Vehicle%20-%20Verticle) 1.0) 1.2)))))))

(fn OnRep_PreEventActive
  (Utilities|FlowControl|Sequence
    (:then_0
      (bind _outfortplayerpawns (Fortnite|GetFortPlayerPawns))
      (for _array_element _outfortplayerpawns
        (if (Pawn|IsPlayerControlled _array_element)
          (Utilities|IsValid (Ability|GetFortAbilitySystemComponent _array_element)
            (:Is%20Valid
              (for _array_element_2 (Variables|Default|GetAreaGameplayEffects)
                (if (Variables|Default|GetPreEventActive)
                  (if (not (Default|DoesPlayerHaveEffect _array_element_2 (Ability|GetFortAbilitySystemComponent _array_element)))
                    (GameplayEffects|ApplyGameplayEffectToSelf :self (Ability|GetFortAbilitySystemComponent _array_element) :GameplayEffectClass _array_element_2 :Level 0.0))
                  (else
                    (GameplayEffects|RemoveActiveGameplayEffectbySourceEffect (Ability|GetFortAbilitySystemComponent _array_element) _array_element_2 0 -1)))))))))
    (:then_1
      (Networking|SwitchHasAuthority
        (:Remote
          (if (Variables|Default|GetShowHUDMessagingForPreEvent)
            (Class|FortControllerComponentTransientQuests|StartQueueforPlayerDuelEvent (Actor|GetComponentbyClass (Fortnite|GetFirstLocalFortPlayerController true) "/Script/FortniteGame.FortControllerComponent_TransientQuests") (Variables|Default|GetPreEventBannerData))))))))

(fn OnRep_ZoneActive
  (if (Variables|Default|GetZoneActive)
    (Development|SetHiddeninGame (Variables|Default|GetAreaVisuals) false false)
    (Collision|SetCollisionEnabled (Variables|Default|GetAreaOfEffect) "QueryOnly")
    (else
      (Development|SetHiddeninGame (Variables|Default|GetAreaVisuals) true false)
      (Collision|SetCollisionEnabled (Variables|Default|GetAreaOfEffect) "NoCollision"))))

(fn ConstructionScript)
