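; Special event utility Blueprint that clears loot from the area covered by its Box component.
; Its DestroyLoot event runs on authority only, using the box's world location and scaled extent for overlap queries.
; It optionally destroys overlapping BuildingContainer actors (on by default), then destroys overlapping FortPickup actors.
; Each pass writes its found count to the shipping log under a PistolCanary tag.
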(event Custom|DestroyLoot
  (params (DestroyContainers :type Boolean :default true))
  (bind _box (Variables|Default|GetBox))
  (bind _returnvalue (Transformation|GetWorldLocation _box))
  (bind _returnvalue_2 (Components|Box|GetScaledBoxExtent _box))
  (Networking|SwitchHasAuthority
    (:Authority
      (if DestroyContainers
        (bind _outactors (Collision|BoxOverlapActors _returnvalue _returnvalue_2 (Utilities|Array|MakeArray "ObjectTypeQuery1") "/Script/FortniteGame.BuildingContainer" :!class /Script/Engine.KismetSystemLibrary)) ; find structures
        (Fortnite|FortShippingLog 0 (Utilities|String|Append "PistolCanary LootDestroyer found containers=" (Utilities|String|ToString(Integer) (Utilities|Array|Length _outactors))))
        (for _array_element _outactors
          (Actor|DestroyActor _array_element))
        (else
          _))
      (bind _outactors_2 (Collision|BoxOverlapActors _returnvalue _returnvalue_2 (Utilities|Array|MakeArray "ObjectTypeQuery4") "/Script/FortniteGame.FortPickup" :!class /Script/Engine.KismetSystemLibrary)) ; find structures
      (Fortnite|FortShippingLog 0 (Utilities|String|Append "PistolCanary LootDestroyer found pickups=" (Utilities|String|ToString(Integer) (Utilities|Array|Length _outactors_2))))
      (for _array_element_2 _outactors_2
        (Actor|DestroyActor _array_element_2)))) ; testing to see if this is necessary
)

(fn ConstructionScript)
