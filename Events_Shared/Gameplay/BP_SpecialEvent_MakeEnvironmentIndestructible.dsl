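; Special event actor that makes building actors inside a box volume indestructible, or destructible again.
; Custom events, optional playlist-ready start, or optional first Box-Playercollision overlap call SetCanDestruct, which caches overlapping buildings and toggles invulnerability, build blocking and damageability.
; ProtectInnerBuildings separately protects buildings in a second box.
; Both box queries skip burst-reaction props and BuildingGameplayActors; a replicated prop count exists only for debug logging.
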(event Custom|PC_Prop_NoDestroy
  (CallFunction|SetCanDestruct true))

(event Custom|PC_Prop_Destroy
  (CallFunction|SetCanDestruct false))

(event OnComponentBeginOverlap(Box-Playercollision)
  (params (OverlappedComponent :type Primitive%20Component%20Object%20Reference) (OtherActor :type Actor%20Object%20Reference) (OtherComp :type Primitive%20Component%20Object%20Reference) (OtherBodyIndex :type Integer) (bFromSweep :type Boolean) (SweepResult :type Hit%20Result%20Structure%20%28by%20ref%29))
  (if (Variables|Default|GetIndestructibleOnPlayerFirstOverlap)
    (Utilities|FlowControl|DoOnce false
      (:Completed
        (Components|Activation|Deactivate (Variables|Default|GetBox%20-%20Player%20collision)) ; this doesn't actually disable it?
        (CallFunction|SetCanDestruct true))) ; Do once when players get near,\nrun the logic again to disable \ndestructibility on the environment
  ))

(event Custom|SetCanDestruct
  (params (Indestructible :type Boolean))
  (bind _cachedbuildingactors (Variables|Default|GetCachedBuildingActors))
  (if (Utilities|Array|IsEmpty (Variables|Default|GetCachedBuildingActors))
    (CallFunction|FindBuildingActors)
    (Utilities|FlowControl|Delay 0.2))
  (Variables|Default|Set%28replicated%29%20total%20props (Utilities|Array|Length _cachedbuildingactors)) ; Replicate this down to clients purely\nfor logging. Trying to figure out why\nthis isn't working on launcher.
  (for _array_element _cachedbuildingactors
    (bind _asbuilding_actor (Utilities|Casting|CastToBuildingActor _array_element))
    (CallFunction|ToggleBuildingDestructability _asbuilding_actor Indestructible)))

(event EventBeginPlay
  (Class|FortAsyncActionCurrentPlaylistReady|PlaylistDataReadyAsync
    (:OnReady
      (if (Variables|Default|GetIndestructibleOnStart)
        (CallFunction|SetCanDestruct true)))))

(event Custom|ProtectInnerBuildings
  (bind _outactors (Collision|ComponentOverlapActors (Variables|Default|GetBox%20-%20IndestructibleSpecial) (Transformation|GetWorldTransform (Variables|Default|GetBox%20-%20IndestructibleSpecial)) (Utilities|Array|MakeArray "ObjectTypeQuery1" "ObjectTypeQuery2" "ObjectTypeQuery6") "/Script/FortniteGame.BuildingActor" (Utilities|Array|MakeArray 0)))
  (for _array_element _outactors
    (bind _asbuilding_actor (Utilities|Casting|CastToBuildingActor _array_element))
    (Utilities|Casting|CastToAthena_Prop_BurstReaction _asbuilding_actor
      (:CastFailed
        (Utilities|Casting|CastToBuildingGameplayActor _asbuilding_actor
          (:CastFailed
            (CallFunction|ToggleBuildingDestructability _asbuilding_actor true)))))))

(fn FindBuildingActors
  (bind _outactors (Collision|ComponentOverlapActors (Variables|Default|GetBox%20-%20ToggleDestructability) (Transformation|GetWorldTransform (Variables|Default|GetBox%20-%20ToggleDestructability)) (Utilities|Array|MakeArray "ObjectTypeQuery1" "ObjectTypeQuery2" "ObjectTypeQuery6") "/Script/FortniteGame.BuildingActor" (Utilities|Array|MakeArray 0)))
  (for _array_element _outactors
    (Utilities|Casting|CastToAthena_Prop_BurstReaction _array_element
      (:CastFailed
        (Utilities|Casting|CastToBuildingGameplayActor _array_element
          (:CastFailed
            (Utilities|Array|Add (Variables|Default|GetCachedBuildingActors) _array_element)))))))

(fn OnRep_%28replicated%29totalprops
  (Development|PrintString (Utilities|String|Append "PistolCanary - total props being made invulnerable: " (Utilities|String|ToString(Integer) (Variables|Default|Get%28replicated%29%20total%20props)))))

(fn ToggleBuildingDestructability
  (params (BuildingActor :type Building%20Actor%20Object%20Reference) (Indestructible :type Boolean))
  (Variables|/Script/FortniteGame.BuildingActor|SetbIsInvulnerable Indestructible BuildingActor)
  (Variables|/Script/FortniteGame.BuildingActor|SetbForceBlockBuildings Indestructible BuildingActor)
  (Variables|/Script/Engine.Actor|SetbCanBeDamaged (not Indestructible) BuildingActor))

(fn ConstructionScript)
