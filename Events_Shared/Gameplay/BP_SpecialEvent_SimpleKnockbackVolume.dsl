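; Special event volume that knocks player pawns away when they hit or overlap its sphere, with custom events to toggle collision and an option to start disabled.
; The launch vector is the normalized actor world location scaled by a distance multiplier, with Z clamped.
; Players in a vortex get ApplyKnockback and a brief vortex exit before re-entry; others get LaunchCharacterJump.
; End overlap only clears a teleport-check timer, which is never set here.
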(event EventBeginPlay
  (if (not (Variables|Default|GetStartEnabled))
    (CallFunction|DisableCollision)))

(event Custom|DisableCollision
  (Collision|SetCollisionEnabled (Variables|Default|GetSphere) "NoCollision"))

(event Custom|EnableKnockbackCollision
  (Collision|SetCollisionEnabled (Variables|Default|GetSphere) "QueryOnly"))

(event OnComponentHit(Sphere)
  (params (HitComponent :type Primitive%20Component%20Object%20Reference) (OtherActor :type Actor%20Object%20Reference) (OtherComp :type Primitive%20Component%20Object%20Reference) (NormalImpulse :type Vector :default "0, 0, 0") (Hit :type Hit%20Result%20Structure%20%28by%20ref%29))
  (bind _asfort_player_pawn (Utilities|Casting|CastToFortPlayerPawn OtherActor))
  (Fortnite|FortShippingLog self "Player hitting knockback volume")
  (CallFunction|KnockbackPlayer _asfort_player_pawn))

(event OnComponentBeginOverlap(Sphere)
  (params (OverlappedComponent :type Primitive%20Component%20Object%20Reference) (OtherActor :type Actor%20Object%20Reference) (OtherComp :type Primitive%20Component%20Object%20Reference) (OtherBodyIndex :type Integer) (bFromSweep :type Boolean) (SweepResult :type Hit%20Result%20Structure%20%28by%20ref%29))
  (bind _asfort_player_pawn (Utilities|Casting|CastToFortPlayerPawn OtherActor))
  (Variables|Default|SetPlayer _asfort_player_pawn)
  (Fortnite|FortShippingLog self "Player overlapping knockback volume")
  (CallFunction|KnockbackPlayer (Variables|Default|GetPlayer)))

(event Custom|KnockbackPlayer
  (params (Player :type Fort%20Pawn%20Object%20Reference))
  (bind _maxzclamp (Variables|Default|GetMaxZClamp))
  (bind _minzclamp (Variables|Default|GetMinZClamp))
  (bind _distancemultiplier (Variables|Default|GetDistanceMultiplier))
  (bind (_returnvalue_x _returnvalue_y _returnvalue_z) (Math|Vector|BreakVector (Math|Vector|MakeVector _distancemultiplier _distancemultiplier _distancemultiplier)))
  (bind _returnvalue (Math|Vector|MakeVector _returnvalue_x _returnvalue_y _returnvalue_z))
  (bind _returnvalue_2 (Transformation|GetActorLocation))
  (bind _returnvalue_3 (Math|Vector|Normalize _returnvalue_2))
  (bind (_returnvalue_x_2 _returnvalue_y_2 _returnvalue_z_2) (Math|Vector|BreakVector (* _returnvalue_3 _returnvalue)))
  (bind _returnvalue_4 (Math|Float|Clamp(Float) _returnvalue_z_2 _minzclamp _maxzclamp))
  (bind _returnvalue_5 (Math|Vector|MakeVector _returnvalue_x_2 _returnvalue_y_2 _returnvalue_4))
  (if (Vortex|IsinVortex Player)
    (HitReaction|ApplyKnockback Player 8e+02 1.5e+02 _returnvalue_5)
    (Vortex|SetinVortex Player false 0.0 -5e+03 0.0 -0.01 0.0 false)
    (Utilities|FlowControl|Delay 0.2)
    (Vortex|SetinVortex Player true 0.0 -5e+03 0.0 -0.01 0.0 false)
    (else
      (Pawn|Character|LaunchCharacterJump Player _returnvalue_5 true true true true))))

(event OnComponentEndOverlap(Sphere)
  (params (OverlappedComponent :type Primitive%20Component%20Object%20Reference) (OtherActor :type Actor%20Object%20Reference) (OtherComp :type Primitive%20Component%20Object%20Reference) (OtherBodyIndex :type Integer))
  (Utilities|Casting|CastToFortPlayerPawn OtherActor)
  (Utilities|Time|ClearandInvalidateTimerbyHandle (Variables|Default|GetTimer_TeleportCheck)))

(fn ConstructionScript)
